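# Masked, lit flipbook material for the Hamster vehicle wheel cosmetic, animating LOD base color and normal textures on skeletal meshes.
# Time scaled by the Animation Phase parameter drives two MF_FlipbookSubUV calls with a sub-image setting of four, keeping albedo and normal in step.
# The albedo flipbook's RGB feeds both base color and opacity mask, alongside scalar metallic and roughness.
# The attributes then pass through the shared vehicle damage and de-rez transition functions.
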Material /VehicleCosmetics/Wheels/Hamster/Materials/M_Wheel_Hamster_FlipBook
graph {
 properties {
  bAutomaticallySetUsageInEditor=false
  bUseMaterialAttributes=true
  bUsedWithMeshDeformer=true
  bUsedWithSkeletalMesh=true
  blendMode=BLEND_Masked
  usedShadingModels=Default Lit
 }
 nodes [
  n1 = Constant() {r=4}
  n2 = Multiply(A=Time(), B=ScalarParameter() {defaultValue=16, group=00_Base, parameterName=Animation Phase}) {constB=6}
  n3 = MaterialFunctionCall!(
   Texture=TextureObjectParameter() {group=00_Base, parameterName=Albedo, sortPriority=0, texture=/VehicleCosmetics/Wheels/Hamster/Textures/T_Hamster_BC_LOD}
   UVs
   SubImages=@n1
   Frame=@n2
  ) {materialFunction=/Game/Global/Materials/Character_Parent_2023/VideoFlipbook_Parent_2025/MaterialFunctions/MF_FlipbookSubUV}
  n4 = TextureObjectParameter() {
   group=00_Base
   parameterName=Normal
   samplerType=SAMPLERTYPE_Normal
   sortPriority=1
   texture=/VehicleCosmetics/Wheels/Hamster/Textures/T_Hamster_N_LOD
  }
  n5 = SetMaterialAttributes!(
   MaterialAttributes
   Base Color=@n3.RGB
   Opacity Mask=@n3.RGB
   Normal=MaterialFunctionCall(Texture=@n4, SubImages=@n1, Frame=@n2) {materialFunction=&MF_FlipbookSubUV}.RGB
   Metallic=ScalarParameter() {group=00_Base, parameterName=Metallic}
   Roughness=ScalarParameter() {defaultValue=0.6, group=00_Base, parameterName=Roughness}
  ) {attributeSetTypes [
    #1
    #2
    #3
    #4
    #5
   ]}
  GetMaterialAttributes(None=MaterialFunctionCall(In=MaterialFunctionCall(Material=@n5) {materialFunction=&MF_BaseVehicle_Damage}.Result) {materialFunction=&MF_VehicleDeRez_Transition}.Result) {attributeGetTypes [#6]}
 ]
}
